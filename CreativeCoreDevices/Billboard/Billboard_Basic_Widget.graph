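# Billboard_Basic_Widget is the Billboard device's UMG widget, showing configurable, auto-wrapped text over a colored Border inside a fixed-size SizeBox.
# InitFromObject casts to Billboard_WidgetComp and binds its update and visibility dispatchers.
# Updates run setters for text, colors, justification, size, font/outline and enum-driven shadow; visibility switches between Hidden and HitTestInvisible.
# Font and size setters rebuild the font struct from current values, resetting spacing and skew.
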WidgetBlueprint /CreativeCoreDevices/Billboard/Billboard_Basic_Widget
graphs [
 EventGraph {nodes [
   n1 = On Visiblity Changed!() -> (OutputDelegate, then, NewVisible:ECreativeVisibilityWithEdit Enum=NewEnumerator0) {is=title K2Node_CustomEvent}
   > SetVisibility(InVisibility=Select(
     Option 0=Hidden
     Option 1=HitTestInvisible
     Index="Not Equal (Enum)"!(A:ECreativeVisibilityWithEdit Enum=@n1.NewVisible, B:ECreativeVisibilityWithEdit Enum=NewEnumerator0) -> (ReturnValue:bool)
    ))
   n2 = On Details Updated!() -> (
    OutputDelegate
    then
    Details_Text_3_#1:Text
    Details_Color_Text_8_#2:Linear Color Structure
    Details_Color_Background_9_#3:Linear Color Structure
    Details_TextJustification_12_#4:ETextJustify Enum
    Details_TextSize_15_#5:int
    Details_OutlineStrength_17_#6:int
    Details_FontFamily_20_#7:Object Reference
    Details_FontTypeface_24_#8:Name
    Details_Shadow_27_#9:EBillboardshadowDirection Enum
    Details_ShadowScale_30_#10:double=~
   ) {is=title K2Node_CustomEvent}
   > SetText!(execute, self:Self Object Reference, Text:String="To String (Text)"(InText=@n2.Details_Text_3_#1)) -> (then)
   > SetTextColor!(execute, self, TextColor:Linear Color Structure=@n2.Details_Color_Text_8_#2) -> (then)
   > SetBackgroundColor!(execute, self, Color:Linear Color Structure=@n2.Details_Color_Background_9_#3) -> (then)
   > SetTextAlignment!(execute, self, Alignment:ETextJustify Enum=@n2.Details_TextJustification_12_#4) -> (then)
   > SetTextScale!(execute, self, FontSize:int=@n2.Details_TextSize_15_#5) -> (then)
   > SetTextFont!(
    execute
    self
    OutlineStrength:int=@n2.Details_OutlineStrength_17_#6
    Font Family:Object Reference=@n2.Details_FontFamily_20_#7
    Font Typeface:Name=@n2.Details_FontTypeface_24_#8
   ) -> (then)
   > SetShadow!(
    execute
    self
    Shadow:EBillboardshadowDirection Enum=@n2.Details_Shadow_27_#9
    ShadowScale:double=@n2.Details_ShadowScale_30_#10
   ) -> (then)
   n3 = Event InitFromObject!() -> (OutputDelegate, then, InitObject:Object Reference) {is=title K2Node_Event}
   > n4 = Cast To Billboard_WidgetComp!(execute, Object=@n3.InitObject) -> (then, CastFailed, AsBillboard Widget Comp:Billboard Widget Comp Object Reference) {is=title K2Node_DynamicCast}
   > Bind Event to On Widget Updated(self:Billboard Widget Comp Object Reference=@n4, Delegate=@n2) {is=title K2Node_AddDelegate}
   > Bind Event to On Widget Visibility Changed(self:Billboard Widget Comp Object Reference=@n4, Delegate=Create Event()) {is=title K2Node_AddDelegate}
  ]}
 SetBackgroundColor {nodes [
   n1 = SetBackgroundColor!() -> (then, Color:Linear Color Structure) {is=title K2Node_FunctionEntry}
   > SetBrushColor(self=Get Border() -> (Border:Border Object Reference), InBrushColor=@n1)
  ]}
 SetShadow {nodes [
   n1 = SetShadow!() -> (then, Shadow:EBillboardshadowDirection Enum=NewEnumerator0, ShadowScale:double=~) {is=title K2Node_FunctionEntry}
   n2 = Get UserText() -> (UserText:Text Object Reference)
   > SetShadowColorAndOpacity!(
    execute
    self:Text Object Reference=@n2
    InShadowColorAndOpacity:Linear Color Structure=SelectColor(
     A=(R=0,G=0,B=0,A=0)
     B=(R=0,G=0,B=0,A=1)
     bPickA="Equal (Enum)"(A:EBillboardshadowDirection Enum=@n1, B:EBillboardshadowDirection Enum=NewEnumerator0)
    )
   ) -> (then)
   n3 = float * float(A=Get ShadowScale() -> (ShadowScale:double), B=-1)
   n4 = Get ShadowScale()
   > SetShadowOffset!(
    execute
    self:Text Object Reference=@n2
    InShadowOffset:Vector 2D Structure=Select!(
     NewEnumerator0:Vector 2D Structure
     NewEnumerator1_X:double=@n4
     NewEnumerator1_Y:double=@n4
     NewEnumerator2_X:double=@n3
     NewEnumerator2_Y:double=@n4
     NewEnumerator3_X:double=@n3
     NewEnumerator3_Y:double=@n3
     NewEnumerator4_X:double=@n4
     NewEnumerator4_Y:double=@n3
     Index:EBillboardshadowDirection Enum=Get Shadow() -> (Shadow:EBillboardshadowDirection Enum=NewEnumerator0)
    ) -> (ReturnValue:Vector 2D Structure)
   ) -> (then)
  ]}
 SetText {nodes [
   n1 = SetText!() -> (then, Text:String) {is=title K2Node_FunctionEntry}
   > "SetText (Text)"(self=Get UserText(), InText="To Text (String)"(InString=@n1))
  ]}
 SetTextAlignment {nodes [
   n1 = SetTextAlignment!() -> (then, Alignment:ETextJustify Enum=Left) {is=title K2Node_FunctionEntry}
   > SetJustification!(execute, self:Text Layout Widget Object Reference=Get UserText(), InJustification:ETextJustify Enum=@n1) -> (then)
  ]}
 SetTextColor {nodes [
   n1 = SetTextColor!() -> (then, TextColor:Linear Color Structure) {is=title K2Node_FunctionEntry}
   > SetColorAndOpacity(self=Get UserText(), InColorAndOpacity=Make Slate Color(SpecifiedColor=@n1))
  ]}
 SetTextFont {nodes [
   n1 = SetTextFont!() -> (then, OutlineStrength:int, Font Family:Object Reference, Font Typeface:Name) {is=title K2Node_FunctionEntry}
   n2 = Get!(self:Text Object Reference=Get UserText()) -> (
    Font_FontObject:Object Reference
    Font_FontMaterial:Object Reference
    Font_OutlineSettings:Font Outline Settings Structure
    Font_TypefaceFontName:Name
    Font_Size:float=~
    Font_LetterSpacing:int
    Font_SkewAmount:float=~
    Font_bForceMonospaced:bool=~
    Font_bMaterialIsStencil:bool=~
    Font_MonospacedWidth:float=~
   )
   > SetFont!(
    execute
    self=Get UserText()
    InFontInfo_FontObject:Object Reference=@n1.Font Family
    InFontInfo_FontMaterial:Object Reference=@n2.Font_FontMaterial
    InFontInfo_OutlineSettings:Font Outline Settings Structure=Make Font Outline Settings!(
     OutlineSize:int=@n1
     bMiteredCorners:bool=False
     bSeparateFillAlpha:bool=False
     bApplyOutlineToDropShadows:bool=true
     OutlineMaterial:Object Reference
     OutlineColor:Linear Color Structure=(R=0,G=0,B=0,A=1)
    ) -> (FontOutlineSettings:Font Outline Settings Structure) {is=title K2Node_MakeStruct}
    InFontInfo_TypefaceFontName:Name=@n1.Font Typeface
    InFontInfo_Size:float=@n2.Font_Size
    InFontInfo_LetterSpacing:int=0
    InFontInfo_SkewAmount:float=0
    InFontInfo_bForceMonospaced:bool=false
    InFontInfo_bMaterialIsStencil:bool=false
    InFontInfo_MonospacedWidth:float=1
   ) -> (then)
  ]}
 SetTextScale {nodes [
   n1 = SetTextScale!() -> (then, FontSize:int) {is=title K2Node_FunctionEntry}
   n2 = Get UserText()
   n3 = Break Slate Font Info!(SlateFontInfo:"Slate Font Info Structure (by ref)"=Get Font!(self:Text Object Reference=@n2) -> (Font:Slate Font Info Structure)) -> (
    FontObject:Object Reference
    FontMaterial:Object Reference
    OutlineSettings:Font Outline Settings Structure
    TypefaceFontName:Name
    Size:float=~
    LetterSpacing:int
    SkewAmount:float=~
    bForceMonospaced:bool=~
    bMaterialIsStencil:bool=~
    MonospacedWidth:float=~
   ) {is=title K2Node_BreakStruct}
   > SetFont!(
    execute
    self=@n2
    InFontInfo:Slate Font Info Structure=MakeSlateFontInfo!(
     FontObject:Object Reference=@n3
     FontMaterial:Object Reference=@n3.FontMaterial
     OutlineSettings:Font Outline Settings Structure=@n3.OutlineSettings
     TypefaceFontName:Name=@n3.TypefaceFontName
     Size:float="To Float (Integer)"(InInt=@n1)
     LetterSpacing:int=0
     SkewAmount:float=0
     bForceMonospaced:bool=false
     bMaterialIsStencil:bool=false
     MonospacedWidth:float=1
    ) -> (ReturnValue:Slate Font Info Structure)
   ) -> (then)
  ]}
]
variables []
widgets {
 info {inheritedWidgetCount=0, parentClass=&UserWidget, rootWidgetClass=&SizeBox, widgetCount=3}
 widgets [
  SizeBox_2 {
   widgetClassPath=&SizeBox
   widget=SizeBox_2
   properties {bOverride_HeightOverride=true, bOverride_WidthOverride=true, heightOverride=1200, widthOverride=2000}
  }
  Border {
   widgetClassPath=&Border
   bIsVariable=true
   parent=SizeBox_2
   widget=Border
   properties {brushColor=<rgba 1 1 1 0>, padding=<ltrb 10 10 10 10>}
   slot {class=SizeBoxSlot}
  }
  UserText {
   widgetClassPath=&TextBlock
   bIsVariable=true
   parent=Border
   widget=UserText
   properties {
    autoWrapText=true
    font {
     bForceMonospaced=false
     bMaterialIsStencil=false
     fontMaterial=null
     fontObject=&Roboto
     letterSpacing=0
     monospacedWidth=1
     size=20
     skewAmount=0
     typefaceFontName=Bold
     outlineSettings {
      bApplyOutlineToDropShadows=false
      bMiteredCorners=false
      bSeparateFillAlpha=false
      outlineMaterial=null
      outlineSize=0
      outlineColor=<rgba 0 0 0 1>
     }
    }
   }
   slot {class=BorderSlot, properties {padding=<ltrb 10 10 10 10>}}
  }
 ]
}
