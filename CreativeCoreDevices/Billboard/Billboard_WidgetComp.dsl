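; Billboard device widget component that renders configurable text on a manually redrawn UMG widget, creating it on the first details update instead of Begin Play.
; The widget class async-loads at runtime and block-loads in editor.
; Updates reach the widget through dispatchers, delayed a tick to avoid redundant refreshes.
; A resolution multiplier scales draw size, inverse component scale and text details, while visibility swaps materials and shrinks hidden render targets.
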(event Custom|Update%20Details
  (params (BillboardDetails :type Struct%20Billboard%20Details%20Structure))
  ;[0 1 0.1429 0.25] Update Text Will update the text on the widget, and also Create it if it needs to/
  ; This also functions as the devices de-facto initialization function (since this will be used on UEFN objects and they can't handle Begin Play)
  (Networking|CanExecuteCosmeticEvents
    (:True
      (Variables|Default|SetBillboardDetails BillboardDetails)
      (CallFunction|CreateOrUpdateWidget))))

(event Custom|Create%20or%20Update%20Widget
  (bind _returnvalue (Editor|InEditor))
  (bind _widget_class (Variables|Default|GetWidget%20Class))
  ;[0 1 0.06881 0.25] Create the widget as we don't have one loaded yet.
  (UserInterface|SetManuallyRedraw (not (Editor|InEditor))) ; Editor has to automatically redraw\nin order for auto-wrap to work.\nNeed to set each time in case value gets stomped by\nUEFN
  (Variables|Internal|GetLoadedUserWidget
    (:then
      (CallFunction|UpdateWidget) ; We already have a valid widget, update immediately.
    )
    (:else
      (if _returnvalue
        (bind _returnvalue_2 (Utilities|LoadClassAssetBlocking _widget_class)) ; If we're editor we can't asnyc load, so block load\n(we're in a component here, so can't use common macro)
        (else
          (bind _class (Utilities|AsyncLoadClassAsset _widget_class)
            (:Completed
              _))))
      (bind _asuser_widget (Utilities|Casting|CastToUserWidgetClass (Utilities|SelectClass _returnvalue_2 _class _returnvalue)))
      (bind _returnvalue_3 (UserInterface|CreateWidget _asuser_widget (Fortnite|GetFirstLocalFortPlayerController true)))
      (bind _output_get (Variables|Internal|SetLoadedUserWidget _returnvalue_3))
      (Init|InitfromObject(Message) _output_get self) ; Binds the widget to this component's Dispatchers.
      (CallFunction|UpdateWidget))))

(event Custom|Update%20Widget
  ;[0 0.5023 1 0.25] Calls InitFromObject on the widget, which will then update itself based on this component (it grabs details from GetDetails). Includes an enforced one frame delay to avoid unnecessary refreshes when it's being initialized (e.g. color and text getting updated in same frame). Also calling SetWidget on next tick so it wont be set back to null when reregistering the component.
  (Utilities|FlowControl|DelayUntilNextTick)
  (Utilities|IsValid (UserInterface|GetUserWidgetObject)
    (:Is%20Valid
      _)
    (:Is%20Not%20Valid
      (bind _loadeduserwidget (Variables|Internal|GetLoadedUserWidget)
        (:then
          (UserInterface|SetWidget _loadeduserwidget))
        (:else
          (Fortnite|FortShippingLog self "Widget failed to load" true)
          (CallFunction|CreateOrUpdateWidget)))))
  (Default|CallOnWidgetUpdated (CallFunction|GetScaledBillboardDetails))
  (UserInterface|RequestRenderUpdate))

(event Custom|Request%20Render%20Update%20Next%20Frame
  ;[0 0.5967 1 0.219] Requests a refresh after a frame delay (to avoid refreshing multiple times in a frame).
  (if (not (Editor|InEditor))
    (Utilities|FlowControl|DelayUntilNextTick))
  (UserInterface|RequestRenderUpdate))

(fn GetScaledBillboardDetails
  (bind _output_get (LocalVariables|SetlocalDetails (Variables|Default|GetBillboardDetails)))
  (bind _structout (Utilities|Struct|SetmembersinStructBillboardDetails _output_get (Math|Float|Round (* (Utilities|Struct|BreakStructBillboardDetails _output_get) (Variables|ComponentSizing|GetResolutionMultiplier))) (Math|Float|Round (* (.OutlineStrength_17_06020145432ACAFFFAC9B9A77BE08490 (Utilities|Struct|BreakStructBillboardDetails _output_get)) (Variables|ComponentSizing|GetResolutionMultiplier))) (* (.ShadowScale_30_004B4E7F424EFAF0EFF6FDAB3A2499E2 (Utilities|Struct|BreakStructBillboardDetails _output_get)) (Variables|ComponentSizing|GetResolutionMultiplier))))
  (return _structout))

(fn SetResolutionMultiplier
  (params (Multiplier :type Float%20%28double-precision%29))
  (bind _output_get (Variables|ComponentSizing|SetResolutionMultiplier Multiplier))
  (Transformation|SetRelativeScale3D (/ (Variables|ComponentSizing|GetBaseScale) _output_get)) ; Updates the scale of the device by by resolution\n(we're going to be making the draw size the inverse of the scale to \nadjust resolution)
  (CallFunction|UpdateWidgetResolution)
  (CallFunction|UpdateWidget))

(fn UpdateVisibilityAndUpdateWidget
  (params (Visibility :type ECreativeVisibilityWithEdit%20Enum :default "NewEnumerator0") (IsDoubleSided :type Boolean))
  (bind _visibilitymaterials (Variables|Default|GetVisibilityMaterials))
  (bind (_value _returnvalue) (Utilities|Map|Find _visibilitymaterials Visibility))
  ;[0 0.7788 1 0.25] Render target space stays allocated for invisible widgets, so when hiding we also set the text to null, which for this widget will release the widget.
  ;;[1 1 1] Determine if we should use Two Sided material or not.
  ;; Two Sided material is only used when the Device is supposed to display Two Sided and it has the background rendered
  (Networking|CanExecuteCosmeticEvents
    (:True
      (if _returnvalue
        (Rendering|Material|SetMaterial 0 (select (and (Utilities|Enum|Equal(Enum) Visibility "NewEnumerator1") IsDoubleSided) (Variables|Default|GetTwoSidedMaterial) _value)) ; We want edit mode to have a unique (semi-disabled) material.\n
        (else
          _))
      (bind _output_get (Variables|Default|SetWidgetVisibility Visibility))
      (Default|CallOnWidgetVisibilityChanged _output_get)
      (CallFunction|UpdateWidgetResolution) ; An invisible widget sets its draw scale to a tiny value to make the render memory negligible.\n(originally I was just throwing away the widget when invisible but having to create and uncreate constantly\nwas causing problems).
      (Rendering|SetCastShadow false) ; Explictly disable shadow casting, otherwise the Two Sided material will cast shadow
    )))

(fn UpdateWidgetResolution
  ;[1 0.803 0 0.1] Updates the resolution of the widget by either the resolution multiplier (if the widget is visible) or to a minimum value set by Invisible Resolution, if it's not.
  ; Editor always wants to be visible, as we won't be hiding the editor visibility.
  (UserInterface|SetDrawSize (select (Utilities|Enum|NotEqual(Enum) (Variables|Default|GetWidgetVisibility) "NewEnumerator0") (Utilities|Operators|Multiply (Variables|ComponentSizing|GetBaseDrawSize) (Math|Vector2D|MakeVector2D (Variables|ComponentSizing|GetResolutionMultiplier) (Variables|ComponentSizing|GetResolutionMultiplier))) (Variables|Default|GetInvisibleResolution)))
  (CallFunction|RequestRenderUpdateNextFrame))
