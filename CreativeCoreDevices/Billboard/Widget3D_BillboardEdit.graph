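# Unlit translucent material for a 3D widget on the Billboard device, apparently its edit variant going by the name.
# It samples the SlateUI texture parameter with clamped shared sampling, passing RGB through as color.
# The OpacityFromTexture scalar blends between full opacity and the texture's alpha, and a constant multiplier scales the result down so the widget appears faded.
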Material /CreativeCoreDevices/Billboard/Widget3D_BillboardEdit
graph {
 properties {blendMode=BLEND_Translucent, pixelDepthOffsetMode=PDOM_Legacy, shadingModel=MSM_Unlit, usedShadingModels=Unlit}
 nodes [
  n1 = TextureSampleParameter2D() {parameterName=SlateUI, samplerSource=SSM_Clamp_WorldGroupSettings, texture=/Engine/EngineResources/WhiteSquareTexture}
  Multiply(
   A=LinearInterpolate(B=@n1.A, Alpha=ScalarParameter() {defaultValue=1, parameterName=OpacityFromTexture, sliderMax=1}) {constA=1}
   B=Constant() {r=0.35}
  )
  Reroute(None=@n1.RGB)
 ]
}
